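# Animation Blueprint for the Dime Aviator male medium body cosmetic, following the parent mesh's pose.
# AnimGraph copies that pose, replaces neck and head rotations with a fixed yaw offset in bone space, then runs the body's Control Rig.
# The dynamics cache only passes the result through.
# EventGraph computes a gravity override via a deprecated function and scaled, negated neck/head deltas from ref pose, none visibly wired into AnimGraph.
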AnimBlueprint /BRCosmetics/Characters/Player/Male/Medium/Bodies/M_MED_DimeAviator/Meshes/M_MED_DimeAviator_AnimBP
target_skeleton=&/M_MED_DimeAviator_Skeleton
graphs [
 AnimGraph {nodes [
   Output Pose(Result=Use cached pose 'DynamicsCache'()) {layerGroup=None}
   n1 = "Transform (Modify) Bone - Bone: neck_01"!(ComponentPose=Local To Component(LocalPose=Copy Pose From Mesh() {bUseAttachedParent=true}), Alpha=1, Rotation="0,-10,0.000") -> (Pose) {
    rotationMode=BMM_Replace
    rotationSpace=BCS_BoneSpace
    alphaBoolBlend {blendOption=Linear}
    alphaScaleBiasClamp {interpSpeedDecreasing=5, interpSpeedIncreasing=20}
    boneToModify {boneName=neck_01}
    rotation {yaw=-10}
   }
   n2 = "Transform (Modify) Bone - Bone: neck_02"!(ComponentPose=@n1, Alpha=1, Rotation="0,-10,0.000") -> (Pose) {
    rotationMode=BMM_Replace
    rotationSpace=BCS_BoneSpace
    alphaBoolBlend {blendOption=Linear}
    alphaScaleBiasClamp {interpSpeedDecreasing=5, interpSpeedIncreasing=20}
    boneToModify {boneName=neck_02}
    rotation {yaw=-10}
   }
   n3 = "Transform (Modify) Bone - Bone: head"!(ComponentPose=@n2, Alpha=1, Rotation="0,-10,0.000") -> (Pose) {
    rotationMode=BMM_Replace
    rotationSpace=BCS_BoneSpace
    alphaBoolBlend {blendOption=Linear}
    alphaScaleBiasClamp {interpSpeedDecreasing=5, interpSpeedIncreasing=20}
    boneToModify {boneName=head}
    rotation {yaw=-10}
   }
   n4 = Control Rig(Source=Component To Local(ComponentPose=@n3)) {
    alphaBoolBlend {blendOption=Linear}
    controlRigAssetReference {blueprintRigClass=&/M_MED_DimeAviator_CtrlRig.M_MED_DimeAviator_CtrlRig_C}
   }
   Save cached pose 'CRCache_Input'(Pose=@n4) {cacheName=CRCache_Input}
   Save cached pose 'DynamicsCache'(Pose=Use cached pose 'CRCache_Input'()) {cacheName=DynamicsCache}
  ]}
 EventGraph {nodes [
   Event BlueprintUpdateAnimation()
   > n1 = Sequence()
   > n2 = Gravity Override State Function DEPRECATED(CharacterPartAnimInstance=Self-Reference(), GameTimeInSeconds=GetGameTimeInSeconds(), GravityOverrideParams=Get Gravity Override Params())
   > Set GravityOverride(GravityOverride=@n2)
   Set Twist Reduction Value(execute=@n1.then_1, Twist Reduction Value:double=-0.25) -> (Output_Get:double)
   > Set Y Reduction Value(Y Reduction Value:double=-0.5) -> (Output_Get:double)
   > Set Z Reduction Value(Z Reduction Value:double=-0.5) -> (Output_Get:double)
   n3 = BreakRotator(InRot=GetDeltaTransformFromRefPose!(self=GetOwnerSkeletalMeshComponent(), BoneName=neck_01, BaseName=spine_05) -> (ReturnValue_Location:Vector, ReturnValue_Rotation:Rotator, ReturnValue_Scale:Vector).ReturnValue_Rotation)
   n4 = Get Twist Reduction Value() -> (Twist Reduction Value:double)
   n5 = Get Y Reduction Value() -> (Y Reduction Value:double)
   n6 = Get Z Reduction Value() -> (Z Reduction Value:double)
   > Set neck01 rot(neck01 rot:Rotator=MakeRotator(
     Roll=float * float(A=@n3, B=@n4) {is=title K2Node_PromotableOperator}
     Pitch=float * float(A=@n3.Pitch, B=@n5)
     Yaw=float * float(A=@n3.Yaw, B=@n6)
    )) -> (Output_Get:Rotator)
   n7 = BreakRotator(InRot=GetDeltaTransformFromRefPose(self=GetOwnerSkeletalMeshComponent(), BoneName=neck_02, BaseName=neck_01).ReturnValue_Rotation)
   n8 = Get Twist Reduction Value()
   n9 = Get Y Reduction Value()
   n10 = Get Z Reduction Value()
   > Set neck02 rot(neck02 rot:Rotator=MakeRotator(Roll=float * float(A=@n7, B=@n8), Pitch=float * float(A=@n7.Pitch, B=@n9), Yaw=float * float(A=@n7.Yaw, B=@n10))) -> (Output_Get:Rotator)
   n11 = BreakRotator(InRot=GetDeltaTransformFromRefPose(self=GetOwnerSkeletalMeshComponent(), BoneName=head, BaseName=neck_02).ReturnValue_Rotation)
   n12 = Get Z Reduction Value()
   n13 = Get Twist Reduction Value()
   n14 = Get Y Reduction Value()
   > Set head rot(head rot:Rotator=MakeRotator(Roll=float * float(A=@n11, B=@n13), Pitch=float * float(A=@n11.Pitch, B=@n14), Yaw=float * float(A=@n11.Yaw, B=@n12))) -> (Output_Get:Rotator)
  ]}
]
variables [
 Gravity Override Params {type=GravityOverrideParamsStruct}
 GravityOverride {type=Vector}
 head rot {type=Rotator}
 neck01 rot {type=Rotator}
 neck02 rot {type=Rotator}
 Twist Reduction Value {type=number}
 Y Reduction Value {type=number}
 Z Reduction Value {type=number}
]
