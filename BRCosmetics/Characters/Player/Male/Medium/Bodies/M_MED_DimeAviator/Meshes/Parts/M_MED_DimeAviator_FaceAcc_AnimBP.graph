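# Animation Blueprint for the Dime Aviator male face accessory, driving its mask, strap and neck-support bones.
# It layers Body_Mesh bones over the parent pose, runs a bone-modify chain, then simulates mask and strap with AnimDynamics.
# The update event caches body socket transforms for pec and neck-support bones.
# Weapon, crouch and targeting events set IsAimingLongGun, blending in fixed mask and strap offsets for rifle-type weapons.
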AnimBlueprint /BRCosmetics/Characters/Player/Male/Medium/Bodies/M_MED_DimeAviator/Meshes/Parts/M_MED_DimeAviator_FaceAcc_AnimBP
target_skeleton=&/M_MED_DimeAviator_FaceAcc_Skeleton
graphs [
 AnimGraph {nodes [
   "Transform (Modify) Bone - Bone: base_maskStrap"!(ComponentPose, bAlphaBoolEnabled:bool=True, Rotation="-5,-18,0") -> (Pose) {
    alphaInputType=Bool
    rotationMode=BMM_Additive
    alphaBoolBlend {blendInTime=0.2, blendOption=Linear, blendOutTime=0.2}
    boneToModify {boneName=base_maskStrap}
    rotation {pitch=-5, yaw=-18}
   }
   "Transform (Modify) Bone - Bone: dyn_maskStrap"!(ComponentPose, Alpha=1, Translation="-0.5,0.75,-1", Rotation="-20,-20,-10") -> (Pose) {
    rotationMode=BMM_Replace
    rotationSpace=BCS_BoneSpace
    translationMode=BMM_Additive
    translationSpace=BCS_BoneSpace
    alphaBoolBlend {blendOption=Linear}
    boneToModify {boneName=dyn_maskStrap}
    rotation=<rot -20 -20 -10>
    translation=<-0.5 0.75 -1>
   }
   Output Pose(Result=Use cached pose 'DynamicsCache'()) {layerGroup=None}
   n1 = "Transform (Modify) Bone - Bone: deform_pec_base_l"!(
    ComponentPose=Local To Component(LocalPose=Use cached pose 'CRCache_Input'())
    bAlphaBoolEnabled:bool=True
    Translation="0,0,0"
    Rotation="0,0,0"
   ) -> (Pose) {
    alphaInputType=Bool
    rotationMode=BMM_Replace
    rotationSpace=BCS_WorldSpace
    translationMode=BMM_Replace
    translationSpace=BCS_ParentBoneSpace
    alphaBoolBlend {blendOption=Linear}
    boneToModify {boneName=deform_pec_base_l}
   }
   n2 = "Transform (Modify) Bone - Bone: deform_pec_l"!(ComponentPose=@n1, bAlphaBoolEnabled:bool=True, Translation="0,0,0", Rotation="0,0,0") -> (Pose) {
    alphaInputType=Bool
    rotationMode=BMM_Replace
    rotationSpace=BCS_ParentBoneSpace
    translationMode=BMM_Replace
    translationSpace=BCS_ParentBoneSpace
    alphaBoolBlend {blendOption=Linear}
    boneToModify {boneName=deform_pec_l}
   }
   n3 = "Transform (Modify) Bone - Bone: base_NeckSupportClav_l"!(ComponentPose=@n2, bAlphaBoolEnabled:bool=True, Translation="0,0,0", Rotation="0,0,0") -> (Pose) {
    alphaInputType=Bool
    rotationMode=BMM_Replace
    rotationSpace=BCS_ParentBoneSpace
    scaleSpace=BCS_ParentBoneSpace
    translationMode=BMM_Replace
    translationSpace=BCS_ParentBoneSpace
    alphaBoolBlend {blendOption=Linear}
    boneToModify {boneName=base_NeckSupportClav_l}
   }
   n4 = "Transform (Modify) Bone - Bone: dyn_NeckSupportClav_l"!(ComponentPose=@n3, bAlphaBoolEnabled:bool=True, Translation="0,0,0", Rotation="0.370,-63.672,11.738") -> (Pose) {
    alphaInputType=Bool
    rotationMode=BMM_Replace
    rotationSpace=BCS_ParentBoneSpace
    scaleSpace=BCS_ParentBoneSpace
    translationMode=BMM_Replace
    translationSpace=BCS_ParentBoneSpace
    alphaBoolBlend {blendOption=Linear}
    boneToModify {boneName=dyn_NeckSupportClav_l}
    rotation=<rot 0.37 -63.672 11.738>
   }
   n5 = "Transform (Modify) Bone - Bone: base_NeckSupportClav_r"!(ComponentPose=@n4, bAlphaBoolEnabled:bool=True, Translation="0,0,0", Rotation="0,0,0") -> (Pose) {
    alphaInputType=Bool
    rotationMode=BMM_Replace
    rotationSpace=BCS_ParentBoneSpace
    scaleSpace=BCS_ParentBoneSpace
    translationMode=BMM_Replace
    translationSpace=BCS_ParentBoneSpace
    alphaBoolBlend {blendOption=Linear}
    boneToModify {boneName=base_NeckSupportClav_r}
   }
   n6 = "Transform (Modify) Bone - Bone: dyn_NeckSupportClav_r"!(ComponentPose=@n5, bAlphaBoolEnabled:bool=True, Translation="0,0,0", Rotation="0,0,0") -> (Pose) {
    alphaInputType=Bool
    rotationMode=BMM_Replace
    rotationSpace=BCS_ParentBoneSpace
    scaleSpace=BCS_ParentBoneSpace
    translationMode=BMM_Replace
    translationSpace=BCS_ParentBoneSpace
    alphaBoolBlend {blendOption=Linear}
    boneToModify {boneName=dyn_NeckSupportClav_r}
   }
   n7 = "Transform (Modify) Bone - Bone: base_NeckSupportFront_l"!(ComponentPose=@n6, bAlphaBoolEnabled:bool=True, Translation="0,0,0", Rotation="0,0,0") -> (Pose) {
    alphaInputType=Bool
    rotationMode=BMM_Replace
    rotationSpace=BCS_ParentBoneSpace
    scaleSpace=BCS_ParentBoneSpace
    translationMode=BMM_Replace
    translationSpace=BCS_ParentBoneSpace
    alphaBoolBlend {blendOption=Linear}
    boneToModify {boneName=base_NeckSupportFront_l}
   }
   n8 = "Transform (Modify) Bone - Bone: dyn_NeckSupportFront_l"!(ComponentPose=@n7, bAlphaBoolEnabled:bool=True, Translation="0,0,0", Rotation="0,0,0") -> (Pose) {
    alphaInputType=Bool
    rotationMode=BMM_Replace
    rotationSpace=BCS_ParentBoneSpace
    scaleSpace=BCS_ParentBoneSpace
    translationMode=BMM_Replace
    translationSpace=BCS_ParentBoneSpace
    alphaBoolBlend {blendOption=Linear}
    boneToModify {boneName=dyn_NeckSupportFront_l}
   }
   n9 = "Transform (Modify) Bone - Bone: base_neckSupport_HlpJnt_bk"!(ComponentPose=@n8, bAlphaBoolEnabled:bool=True, Translation="0,0,0", Rotation="0,0,0") -> (Pose) {
    alphaInputType=Bool
    rotationMode=BMM_Replace
    rotationSpace=BCS_ParentBoneSpace
    scaleSpace=BCS_ParentBoneSpace
    translationMode=BMM_Replace
    translationSpace=BCS_ParentBoneSpace
    alphaBoolBlend {blendOption=Linear}
    boneToModify {boneName=base_neckSupport_HlpJnt_bk}
   }
   n10 = "Transform (Modify) Bone - Bone: dyn_neckSupport_HlpJnt_bk"!(ComponentPose=@n9, bAlphaBoolEnabled:bool=True, Translation="0,0,0", Rotation="0,0,0") -> (Pose) {
    alphaInputType=Bool
    rotationMode=BMM_Replace
    rotationSpace=BCS_ParentBoneSpace
    scaleSpace=BCS_ParentBoneSpace
    translationMode=BMM_Replace
    translationSpace=BCS_ParentBoneSpace
    alphaBoolBlend {blendOption=Linear}
    boneToModify {boneName=dyn_neckSupport_HlpJnt_bk}
   }
   Save cached pose 'CRCache_TransferTransforms'(Pose=Component To Local(ComponentPose=@n10)) {cacheName=CRCache_TransferTransforms}
   n11 = "Transform (Modify) Bone - Bone: dyn_maskStrap"(ComponentPose=Local To Component(LocalPose=Use cached pose 'FinalDynamics'()), Alpha=1, Translation="-4,-1,-3.5", Rotation="-10,-20,-10") {
    rotationMode=BMM_Replace
    rotationSpace=BCS_BoneSpace
    translationMode=BMM_Additive
    translationSpace=BCS_BoneSpace
    alphaBoolBlend {blendOption=Linear}
    boneToModify {boneName=dyn_maskStrap}
    rotation=<rot -10 -20 -10>
    translation=<-4 -1 -3.5>
   }
   n12 = "Transform (Modify) Bone - Bone: dyn_mask"!(ComponentPose=@n11, Translation="-0.25,0,1.5", Rotation="0,27,30", Scale="1,1,1") -> (Pose) {
    rotationMode=BMM_Additive
    rotationSpace=BCS_BoneSpace
    translationMode=BMM_Additive
    translationSpace=BCS_BoneSpace
    alphaBoolBlend {blendOption=Linear}
    boneToModify {boneName=dyn_mask}
    rotation {roll=30, yaw=27}
    translation {x=-0.25, z=1.5}
   }
   Save cached pose 'DynamicsCache'(Pose=Blend Poses by bool(
     BlendPose_0=Component To Local(ComponentPose=@n12)
     BlendPose_1=Use cached pose 'FinalDynamics'()
     BlendTime_0=0.15
     BlendTime_1=0.5
     bActiveValue=Get IsAimingLongGun() -> (IsAimingLongGun:bool)
    ) {blendType=CircularOut}) {cacheName=DynamicsCache}
   n13 = Layered blend per bone(
    BasePose=Copy Pose From Mesh() {bUseAttachedParent=true}
    BlendPoses_0=Copy Pose From Mesh(SourceMeshComponent=Get Body_Mesh()) {bUseAttachedParent=true}
   ) {layerSetup [{branchFilters [
       {boneName=deform_pec_base_l}
       {boneName=deform_pec_l}
       {boneName=base_NeckSupportClav_l}
       {boneName=dyn_NeckSupportClav_l}
       {boneName=base_NeckSupportClav_r}
       {boneName=dyn_NeckSupportClav_r}
       {boneName=base_NeckSupportFront_I}
       {boneName=dyn_NeckSupportFront_I}
       {boneName=dyn_maskStrap}
       {boneName=dyn_mask}
       {boneName=base_neckSupport_HlpJnt_bk}
       {boneName=dyn_neckSupport_HlpJnt_bk}
      ]}]}
   Save cached pose 'CRCache_Input'(Pose=@n13) {cacheName=CRCache_Input}
   n14 = Anim Dynamics - Bone: dyn_maskStrap!(
    ComponentPose=Local To Component(LocalPose=Use cached pose 'CRCache_TransferTransforms'())
    Alpha=1
    GravityOverride:Vector="0,0,-980"
    AngularSpringConstant:float=15000
   ) -> (Pose) {
    angularDampingOverride=1
    angularSpringConstant=15000
    bAngularSpring=true
    bOverrideAngularDamping=true
    bUseGravityOverride=true
    gravityScale=0.35
    simulationSpace=World
    alphaBoolBlend {blendOption=Linear}
    boundBone {boneName=dyn_maskStrap}
    componentAppliedLinearAccClamp=<10000 10000 10000>
    componentLinearAccScale=<1 1 1>
    componentLinearVelScale=<1 1 1>
    gravityOverride {z=-980}
    physicsBodyDefinitions [{
      boundBone {boneName=dyn_maskStrap}
      boxExtents=<20 30 30>
      constraintSetup {angularLimitsMax=<0 10 0>, angularLimitsMin=<0 -8 0>, angularTarget=<1 0 0>}
      localJointOffset=<0.5 0 0>
     }]
   }
   n15 = Anim Dynamics - Bone: dyn_mask!(ComponentPose=@n14, Alpha=1, GravityOverride:Vector="0,0,-980") -> (Pose) {
    angularDampingOverride=1
    angularSpringConstant=15000
    bAngularSpring=true
    bOverrideAngularDamping=true
    bUseGravityOverride=true
    gravityScale=0.35
    simulationSpace=World
    alphaBoolBlend {blendOption=Linear}
    boundBone {boneName=dyn_mask}
    componentAppliedLinearAccClamp=<10000 10000 10000>
    componentLinearAccScale=<1 1 1>
    componentLinearVelScale=<1 1 1>
    gravityOverride {z=-980}
    physicsBodyDefinitions [{
      boundBone {boneName=dyn_mask}
      boxExtents=<20 30 30>
      constraintSetup {angularLimitsMax=<0 10 0>, angularLimitsMin=<0 -8 0>, angularTarget=<1 0 0>}
      localJointOffset=<0.5 0 0>
     }]
   }
   n16 = Anim Dynamics - Bone: dyn_maskStrap(
    ComponentPose=Local To Component(LocalPose=Use cached pose 'CRCache_TransferTransforms'())
    Alpha=0.3
    GravityOverride="0,0,-980"
    AngularSpringConstant=15000
   ) {
    angularDampingOverride=1
    angularSpringConstant=15000
    bAngularSpring=true
    bOverrideAngularDamping=true
    bUseGravityOverride=true
    gravityScale=0.35
    simulationSpace=World
    alphaBoolBlend {blendOption=Linear}
    boundBone {boneName=dyn_maskStrap}
    componentAppliedLinearAccClamp=<10000 10000 10000>
    componentLinearAccScale=<1 1 1>
    componentLinearVelScale=<1 1 1>
    gravityOverride {z=-980}
    physicsBodyDefinitions [{
      boundBone {boneName=dyn_maskStrap}
      boxExtents=<20 30 30>
      constraintSetup {angularLimitsMax=<0 10 0>, angularLimitsMin=<0 -8 0>, angularTarget=<1 0 0>}
      localJointOffset=<0.5 0 0>
     }]
   }
   n17 = Anim Dynamics - Bone: dyn_mask(ComponentPose=@n16, Alpha=0.3, GravityOverride="0,0,-980") {
    angularDampingOverride=1
    angularSpringConstant=15000
    bAngularSpring=true
    bOverrideAngularDamping=true
    bUseGravityOverride=true
    gravityScale=0.35
    simulationSpace=World
    alphaBoolBlend {blendOption=Linear}
    boundBone {boneName=dyn_mask}
    componentAppliedLinearAccClamp=<10000 10000 10000>
    componentLinearAccScale=<1 1 1>
    componentLinearVelScale=<1 1 1>
    gravityOverride {z=-980}
    physicsBodyDefinitions [{
      boundBone {boneName=dyn_mask}
      boxExtents=<20 30 30>
      constraintSetup {angularLimitsMax=<0 10 0>, angularLimitsMin=<0 -8 0>, angularTarget=<1 0 0>}
      localJointOffset=<0.5 0 0>
     }]
   }
   Save cached pose 'FinalDynamics'(Pose=Blend Poses by bool(BlendPose_0=Component To Local(ComponentPose=@n17), BlendPose_1=Component To Local(ComponentPose=@n15), BlendTime_0=0.15, BlendTime_1=0.5) {blendType=CircularOut}) {cacheName=FinalDynamics}
  ]}
 EventGraph {nodes [
   Event BlueprintUpdateAnimation()
   > n1 = Gravity Override State Function DEPRECATED(CharacterPartAnimInstance=Self-Reference(), GameTimeInSeconds=GetGameTimeInSeconds(), GravityOverrideParams=Get Gravity Override Params())
   > Set GravityOverride(GravityOverride=@n1)
   n2 = Get Body_Mesh()
   > Is Valid(InputObject=@n2)
   n3 = GetSocketBoneName!(self:Skinned Mesh Component Object Reference=@n2, InSocketName:Name=upperarm_r) -> (ReturnValue:Name=None)
   > n4 = Set isBodyAttached(isBodyAttached:bool="Not Equal Exactly (String)"(A="To String (Name)"(InName=@n3), B=None)) -> (Output_Get:bool)
   > Branch(Condition=@n4)
   n5 = BreakTransform(InTransform=GetSocketTransform(self=Get Body_Mesh(), InSocketName=deform_pec_base_l, TransformSpace=RTS_ParentBoneSpace))
   > Set deform_pec_base_l_Translation(deform_pec_base_l_Translation:Vector=@n5) -> (Output_Get:Vector)
   > Set deform_pec_base_l_Rotation(deform_pec_base_l_Rotation:Rotator=@n5.Rotation) -> (Output_Get:Rotator)
   n6 = BreakTransform(InTransform=GetSocketTransform(self=Get Body_Mesh(), InSocketName=deform_pec_l, TransformSpace=RTS_ParentBoneSpace))
   > Set deform_pec_l_Translation(deform_pec_l_Translation:Vector=@n6) -> (Output_Get:Vector)
   > Set deform_pec_l_Rotation(deform_pec_l_Rotation:Rotator=@n6.Rotation) -> (Output_Get:Rotator)
   n7 = BreakTransform(InTransform=GetSocketTransform(self=Get Body_Mesh(), InSocketName=base_NeckSupportClav_l, TransformSpace=RTS_ParentBoneSpace))
   > Set base_NeckSupportClav_l_Translation(base_NeckSupportClav_l_Translation:Vector=@n7) -> (Output_Get:Vector)
   > Set base_NeckSupportClav_l_Rotation(base_NeckSupportClav_l_Rotation:Rotator=@n7.Rotation) -> (Output_Get:Rotator)
   n8 = BreakTransform(InTransform=GetSocketTransform(self=Get Body_Mesh(), InSocketName=dyn_NeckSupportClav_l, TransformSpace=RTS_ParentBoneSpace))
   > Set dyn_NeckSupportClav_l_Translation(dyn_NeckSupportClav_l_Translation:Vector=@n8) -> (Output_Get:Vector)
   > Set dyn_NeckSupportClav_l_Rotation(dyn_NeckSupportClav_l_Rotation:Rotator=@n8.Rotation) -> (Output_Get:Rotator)
   n9 = BreakTransform(InTransform=GetSocketTransform(self=Get Body_Mesh(), InSocketName=base_NeckSupportClav_r, TransformSpace=RTS_ParentBoneSpace))
   > Set base_NeckSupportClav_r_Translation(base_NeckSupportClav_r_Translation:Vector=@n9) -> (Output_Get:Vector)
   > Set base_NeckSupportClav_r_Rotation(base_NeckSupportClav_r_Rotation:Rotator=@n9.Rotation) -> (Output_Get:Rotator)
   n10 = BreakTransform(InTransform=GetSocketTransform(self=Get Body_Mesh(), InSocketName=dyn_NeckSupportClav_r, TransformSpace=RTS_ParentBoneSpace))
   > Set dyn_NeckSupportClav_r_Translation(dyn_NeckSupportClav_r_Translation:Vector=@n10) -> (Output_Get:Vector)
   > Set dyn_NeckSupportClav_r_Rotation(dyn_NeckSupportClav_r_Rotation:Rotator=@n10.Rotation) -> (Output_Get:Rotator)
   n11 = BreakTransform(InTransform=GetSocketTransform(self=Get Body_Mesh(), InSocketName=base_NeckSupportFront_l, TransformSpace=RTS_ParentBoneSpace))
   > Set base_NeckSupportFront_I_Translation(base_NeckSupportFront_I_Translation:Vector=@n11) -> (Output_Get:Vector)
   > Set base_NeckSupportFront_I_Rotation(base_NeckSupportFront_I_Rotation:Rotator=@n11.Rotation) -> (Output_Get:Rotator)
   n12 = BreakTransform(InTransform=GetSocketTransform(self=Get Body_Mesh(), InSocketName=dyn_NeckSupportFront_l, TransformSpace=RTS_ParentBoneSpace))
   > Set dyn_NeckSupportFront_I_Translation(dyn_NeckSupportFront_I_Translation:Vector=@n12) -> (Output_Get:Vector)
   > Set dyn_NeckSupportFront_I_Rotation(dyn_NeckSupportFront_I_Rotation:Rotator=@n12.Rotation) -> (Output_Get:Rotator)
   n13 = BreakTransform(InTransform=GetSocketTransform(self=Get Body_Mesh(), InSocketName=base_neckSupport_HlpJnt_bk, TransformSpace=RTS_ParentBoneSpace))
   > Set base_neckSupport_HlpJnt_bk_Translation(base_neckSupport_HlpJnt_bk_Translation:Vector=@n13) -> (Output_Get:Vector)
   > Set base_neckSupport_HlpJnt_bk_Rotation(base_neckSupport_HlpJnt_bk_Rotation:Rotator=@n13.Rotation) -> (Output_Get:Rotator)
   n14 = BreakTransform(InTransform=GetSocketTransform(self=Get Body_Mesh(), InSocketName=dyn_neckSupport_HlpJnt_bk, TransformSpace=RTS_ParentBoneSpace))
   > Set dyn_neckSupport_HlpJnt_bk_Translation(dyn_neckSupport_HlpJnt_bk_Translation:Vector=@n14) -> (Output_Get:Vector)
   > Set dyn_neckSupport_HlpJnt_bk_Rotation(dyn_neckSupport_HlpJnt_bk_Rotation:Rotator=@n14.Rotation) -> (Output_Get:Rotator)
   Event OnBeginCrouching()
   > n15 = Set IsAimingLongGun(IsAimingLongGun:bool=AND Boolean(A=Get bIsCrouching(), B=Get IsLongGunOut() -> (IsLongGunOut:bool))) -> (Output_Get:bool)
   n16 = NOT Boolean(A=@n15)
   Set maskAlpha(
    execute=[@n17.then, @n15.then]
    maskAlpha:double=float * float(A="To Float (Boolean)"(InBool=@n16), B=0.5) {is=title K2Node_PromotableOperator}
   ) -> (Output_Get:double)
   Event OnEndCrouching()
   > n17 = Set IsAimingLongGun(IsAimingLongGun=AND Boolean(A=Get IsLongGunOut(), B=Get AimingDown() -> (AimingDown:bool)))
   Event OnWeaponChanged()
   > n18 = Sequence()
   > Set IsLongGunOut(IsLongGunOut:bool=false) -> (Output_Get:bool)
   > n19 = Switch on EFortWeaponCoreAnimation!(execute, Selection:EFortWeaponCoreAnimation Enum=Get WeaponCoreAnim!() -> (WeaponCoreAnim)) -> (
    Melee:Exec
    Pistol:Exec
    Shotgun:Exec
    PaperBlueprint:Exec
    Rifle:Exec
    MeleeOneHand:Exec
    MachinePistol:Exec
    RocketLauncher:Exec
    GrenadeLauncher:Exec
    GoingCommando:Exec
    AssaultRifle:Exec
    TacticalShotgun:Exec
    SniperRifle:Exec
    TrapPlacement:Exec
    ShoulderLauncher:Exec
    AbilityDecoTool:Exec
    Crossbow:Exec
    C4:Exec
    RemoteControl:Exec
    DualWield:Exec
    AR_BullPup:Exec
    AR_ForwardGrip:Exec
    MedPackPaddles:Exec
    SMG_P90:Exec
    AR_DrumGun:Exec
    Consumable_Small:Exec
    Consumable_Large:Exec
    Balloon:Exec
    MountedTurret:Exec
    CreativeTool:Exec
    ExplosiveBow:Exec
    AshtonIndigo:Exec
    AshtonChicago:Exec
    MeleeDualWield:Exec
    Unarmed:Exec
   )
   Set IsLongGunOut(execute=[@n19.AssaultRifle, @n19.Rifle, @n19.SMG_P90], IsLongGunOut=true)
   Set IsAimingLongGun(execute=@n18.then_1, IsAimingLongGun=AND Boolean(A=Get IsLongGunOut(), B=Get bIsCrouching()))
   Event OnBeginTargeting() {is=title K2Node_Event}
   > Set AimingDown(AimingDown:bool=true) -> (Output_Get:bool)
   > Branch(Condition=Get IsLongGunOut())
   > n20 = Set IsAimingLongGun(IsAimingLongGun=true)
   n21 = Set IsAimingLongGun(execute=@n24.then, IsAimingLongGun=AND Boolean(A=Get IsLongGunOut(), B=Get bIsCrouching()))
   n22 = NOT Boolean(A=@n21)
   Set maskAlpha(execute=[@n20.then, @n21.then], maskAlpha=float * float(A="To Float (Boolean)"(InBool=@n22), B=0.5))
   n23 = Event OnCharacterCustomizationLoaded()
   > Set Body_Mesh(Body_Mesh=Get Array Item with EFort Custom Part Type(Array=@n23.CustomMeshParts))
   Event OnEndTargeting() {is=title K2Node_Event}
   > n24 = Set AimingDown(AimingDown=false)
   "Gravity Override State Function (Parameter Asset) (Nativized)"(GameTimeInSeconds=0)
  ]}
]
variables [
 AimingDown {type=boolean}
 base_neckSupport_HlpJnt_bk_Rotation {type=Rotator}
 base_neckSupport_HlpJnt_bk_Translation {type=Vector}
 base_NeckSupportClav_l_Rotation {type=Rotator}
 base_NeckSupportClav_l_Translation {type=Vector}
 base_NeckSupportClav_r_Rotation {type=Rotator}
 base_NeckSupportClav_r_Translation {type=Vector}
 base_NeckSupportFront_I_Rotation {type=Rotator}
 base_NeckSupportFront_I_Translation {type=Vector}
 Body_Mesh {type=SkeletalMeshComponent}
 deform_pec_base_l_Rotation {type=Rotator}
 deform_pec_base_l_Translation {type=Vector}
 deform_pec_l_Rotation {type=Rotator}
 deform_pec_l_Translation {type=Vector}
 dyn_neckSupport_HlpJnt_bk_Rotation {type=Rotator}
 dyn_neckSupport_HlpJnt_bk_Translation {type=Vector}
 dyn_NeckSupportClav_l_Rotation {type=Rotator}
 dyn_NeckSupportClav_l_Translation {type=Vector}
 dyn_NeckSupportClav_r_Rotation {type=Rotator}
 dyn_NeckSupportClav_r_Translation {type=Vector}
 dyn_NeckSupportFront_I_Rotation {type=Rotator}
 dyn_NeckSupportFront_I_Translation {type=Vector}
 Gravity Override Params {type=GravityOverrideParamsStruct}
 GravityOverride {type=Vector}
 head rot {type=Rotator}
 IsAimingLongGun {type=boolean}
 isBodyAttached {type=boolean}
 IsLongGunOut {type=boolean}
 maskAlpha {type=number}
 neck01 rot {type=Rotator}
 neck02 rot {type=Rotator}
 Twist Reduction Value {type=number, default=0.5}
 Y Reduction Value {type=number}
 Z Reduction Value {type=number}
]
